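# Prototype Niagara flare for the statue in the RedLight temple puzzle: layered glow sprites, a spinning flare and a red light.
# Glow draws three camera-offset sprite layers sized by User.OverallSize, with time-of-day alpha faded by User.CubeninjaFX_Alpha and sine pulses on two layers.
# SingleLoopingParticle rotates a scaled sprite plus halo, its color lerped by User.IsAlt.
# The light emitter uses spawn-group masks on two bursts to vary sine-modulated red intensity.
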NiagaraSystem /ChronoEnviroItems/TemplePuzzles/RedLight/Assets/VFX/Proto/NS_Proto_Statue_Flare
dependencies {
 usedDataInterfaces [&NiagaraDataInterfaceCurve, &NiagaraDataInterfaceVectorCurve]
 usedDynamicInputs []
 usedModules [&SystemState]
}
emitters {
 Glow {
  emitterUpdateScript [
   &EmitterState(Life Cycle Mode=::NewEnumerator1, Loop Behavior=::NewEnumerator1, Loop Duration Mode=::NewEnumerator1, UseLoopDelay=false)
   &SpawnBurst_Instantaneous(Spawn Count=1, Use Spawn Probability=false)
  ]
  particleSpawnScript [
   /Niagara/Modules/Spawn/Initialization/InitializeParticle(
    Color=<rgba 1 0.0308 0 0.05>
    Module.Write Color=true
    Module.Write Lifetime=true
    Module.Write Mass=true
    Module.Write Material Random=true
    Module.Write Position=false
    Module.Write Scale=false
    Module.Write SpriteRotation=false
    Module.Write SpriteSize=true
    Position=&SimulationPosition()
    Sprite Size=&Vector2DFromFloat(Value=&Multiply_Float(A=1, B=$User.OverallSize))
   )
   &CameraOffset(Camera Offset Amount=30)
   /SetVariables(Particles.SpriteSize2=&Vector2DFromFloat(Value=&Multiply_Float(A=10, B=$User.OverallSize))) {isSetParameters=true}
   /SetVariables(Particles.SpriteSize3:Vector2f=&Vector2DFromFloat(Value=&Multiply_Float(A=3.5, B=$User.OverallSize))) {isSetParameters=true}
  ]
  particleUpdateScript [
   &ParticleState(Kill Particles When Lifetime Has Elapsed=false, Let Infinitely Lived Particles Die When Emitter Deactivates=true)
   &SolveForcesAndVelocity(Clamp Velocity=false, Limit Acceleration=false)
   &ScaleColor(
    Scale Alpha=&Multiply_Float(
     A=&FNTimeOfDayScalar(Night Scale=1, Morning Scale=5, Day Scale=7, Evening Scale=0.3)
     B=&OneMinusFloat(Float=$User.CubeninjaFX_Alpha)
    )
    Scale RGB=&VectorFromCurve(VectorCurve=DataInterface() {propertyValues {
       XCurve {keys [
         {interpMode=RCIM_Cubic, tangentMode=RCTM_User, time=0, value=0}
         {interpMode=RCIM_Cubic, tangentMode=RCTM_User, time=0.5, value=1}
         {tangentMode=RCTM_User, time=1, value=0}
        ]}
       YCurve {keys [
         {interpMode=RCIM_Cubic, tangentMode=RCTM_User, time=0, value=0}
         {interpMode=RCIM_Cubic, tangentMode=RCTM_User, time=0.5, value=1}
         {tangentMode=RCTM_User, time=1, value=0}
        ]}
       ZCurve {keys [
         {interpMode=RCIM_Cubic, tangentMode=RCTM_User, time=0, value=0}
         {interpMode=RCIM_Cubic, tangentMode=RCTM_User, time=0.5, value=1}
         {tangentMode=RCTM_User, time=1, value=0}
        ]}
      }})
    ScaleA=true
    ScaleRGB=true
   ) {module=ScaleColor001}
   /SetVariables(Particles.Color2=&MultiplyLinearColor(
     Linear Color=<rgba 1.5 0.353 1.3508 1>
     Scale Factor=&MakeFloat4FromVectorAndFloat(
      "Vector (RGB)"=&VectorFromCurve(VectorCurve=DataInterface() {propertyValues {
         XCurve {keys [
           {interpMode=RCIM_Cubic, tangentMode=RCTM_User, time=0, value=0}
           {interpMode=RCIM_Cubic, tangentMode=RCTM_User, time=0.5, value=1}
           {tangentMode=RCTM_User, time=1, value=1}
          ]}
         YCurve {keys [
           {interpMode=RCIM_Cubic, tangentMode=RCTM_User, time=0, value=0}
           {interpMode=RCIM_Cubic, tangentMode=RCTM_User, time=0.5, value=1}
           {tangentMode=RCTM_User, time=1, value=1}
          ]}
         ZCurve {keys [
           {interpMode=RCIM_Cubic, tangentMode=RCTM_User, time=0, value=0}
           {interpMode=RCIM_Cubic, tangentMode=RCTM_User, time=0.5, value=1}
           {tangentMode=RCTM_User, time=1, value=1}
          ]}
        }})
      "Float (Alpha)"=&Multiply_Float(
       A=&Multiply_Float(
        A=&OneMinusFloat(Float=$User.CubeninjaFX_Alpha)
        B=&FNTimeOfDayScalar(Night Scale=0.3, Morning Scale=0.05, Day Scale=0.05, Evening Scale=0.3)
       )
       B=&Multiply_Float(A=1, B=&Sine(Normalized Angle=$Emitter.Age, Period=1.2, Scale=5, Bias=5))
      )
     )
    )) {isSetParameters=true}
   /SetVariables(Particles.Color3:LinearColor=&MultiplyLinearColor(
     Linear Color=<rgba 2 0 0 1>
     Scale Factor=&MakeFloat4FromVectorAndFloat(
      "Vector (RGB)"=&VectorFromCurve(VectorCurve=DataInterface() {propertyValues {
         XCurve {keys [
           {interpMode=RCIM_Cubic, tangentMode=RCTM_User, time=0, value=0}
           {interpMode=RCIM_Cubic, tangentMode=RCTM_User, time=0.5, value=1}
           {tangentMode=RCTM_User, time=1, value=1}
          ]}
         YCurve {keys [
           {interpMode=RCIM_Cubic, tangentMode=RCTM_User, time=0, value=0}
           {interpMode=RCIM_Cubic, tangentMode=RCTM_User, time=0.5, value=1}
           {tangentMode=RCTM_User, time=1, value=1}
          ]}
         ZCurve {keys [
           {interpMode=RCIM_Cubic, tangentMode=RCTM_User, time=0, value=0}
           {interpMode=RCIM_Cubic, tangentMode=RCTM_User, time=0.5, value=1}
           {tangentMode=RCTM_User, time=1, value=1}
          ]}
        }})
      "Float (Alpha)"=&Multiply_Float(
       A=&Multiply_Float(
        A=&OneMinusFloat(Float=$User.CubeninjaFX_Alpha)
        B=&FNTimeOfDayScalar(Night Scale=4, Morning Scale=0.3, Evening Scale=4)
       )
       B=&Sine(Normalized Angle=$Emitter.Age, Period=5, Scale=1, Bias=1)
      )
     )
    )) {isSetParameters=true}
   /SetVariables(Particles.CameraOffset2=20) {isSetParameters=true}
   /SetVariables(Particles.CameraOffset3=-10) {isSetParameters=true}
  ]
  rendererClasses [&NiagaraSpriteRendererProperties]
  renderers [
   {rendererClass=&NiagaraSpriteRendererProperties}
   {rendererClass=&NiagaraSpriteRendererProperties, rendererIndex=1}
   {rendererClass=&NiagaraSpriteRendererProperties, rendererIndex=2}
  ]
 }
 SingleLoopingParticle {
  emitterUpdateScript [
   &EmitterState()
   &SpawnBurst_Instantaneous(Use Spawn Probability=false, UseLoopCountLimit=false)
  ]
  particleSpawnScript [
   /Niagara/Modules/Spawn/Initialization/V2/InitializeParticle(
    Color=&LerpLinearColors(StartColor=<rgba 5 0.362 4.6182 1>, EndColor=<rgba 4 2.8613 0.584 1>, LerpFactor=$User.IsAlt)
    Lifetime=2
    Sprite Rotation Angle=-5
    Sprite Rotation Mode=::NewEnumerator0
    Sprite Size Mode=::NewEnumerator3
    Uniform Sprite Size=35
    UsePositionOffset=false
    Write Material Random=true
   )
   &CameraOffset(Camera Offset Amount=20)
  ]
  particleUpdateScript [
   &ParticleState()
   &SpriteRotationRate(Rotation Rate=30)
   &ScaleSpriteSize(Uniform Curve Sprite Scale=DataInterface() {propertyValues {Curve {keys [
        {arriveTangent=9.4013, interpMode=RCIM_Cubic, leaveTangent=9.4013, tangentMode=RCTM_User, time=0, value=0}
        {arriveTangent=-1.1813, interpMode=RCIM_Cubic, leaveTangent=-1.1813, tangentMode=RCTM_User, time=0.1408, value=0.9847}
        {arriveTangent=-0.0405, interpMode=RCIM_Cubic, leaveTangent=-0.0405, tangentMode=RCTM_User, time=1, value=0.65}
       ]}}})
   &SolveForcesAndVelocity(Clamp Velocity=false, Limit Acceleration=false)
   &ScaleColor(ScaleA=true, ScaleRGB=true)
   /SetVariables(
    Particles.Halo Size:Vector2f=&Multiply_Vector2D_ByFloat(Float=3, Vector2D=$Particles.SpriteSize)
    Particles.New Linear Color:LinearColor=&MultiplyLinearColor(Linear Color=$Particles.Color, Scale Factor=<1 1 1 0.05>)
   ) {isSetParameters=true}
  ]
  rendererClasses [&NiagaraSpriteRendererProperties]
  renderers [
   {rendererClass=&NiagaraSpriteRendererProperties}
   {rendererClass=&NiagaraSpriteRendererProperties, rendererIndex=1}
  ]
 }
 light {
  emitterUpdateScript [
   &EmitterState(Life Cycle Mode=::NewEnumerator1, Loop Behavior=::NewEnumerator1, Loop Duration=1, UseLoopDelay=false)
   &SpawnBurst_Instantaneous(Spawn Count=1, Use Spawn Probability=false)
   &SpawnBurst_Instantaneous(Spawn Count=1, Spawn Group=1, Spawn Time=1, Use Spawn Probability=false) {module=SpawnBurst_Instantaneous001}
  ]
  particleSpawnScript [&InitializeParticle(Color=<rgba 5 0 0 1>, Lifetime=0.5, Position Offset=<0 -8 0>, UsePositionOffset=true, Write Material Random=true)]
  particleUpdateScript [
   &ParticleState(Kill Particles When Lifetime Has Elapsed=false)
   &ScaleColor(
    Scale Alpha=&Multiply_Float(A=$User.Light, B=&FNTimeOfDayScalar(Night Scale=0.5, Morning Scale=0.3, Evening Scale=0.3))
    Scale RGB=<4000 4000 4000>
    ScaleA=true
    ScaleRGB=true
   ) {enabled=false, module=ScaleColor002}
   /SetVariables(Particles.Age:NiagaraFloat=&MaskFloatBySpawnGroup(Masked Float=$Particles.Age, Passthrough Float=0, Spawn Group Mask=$Emitter.SpawnGroup)) {isSetParameters=true}
   &Color(Color=&MultiplyLinearColor(
     Linear Color=<rgba 400 0 0 1>
     Scale Factor=&MakeVector4_FromFloat(Float=&MaskFloatBySpawnGroup(
       Masked Float=&FloatFromCurve(
        FloatCurve=DataInterface() {propertyValues {Curve {keys [<at 0 0>,<at 1 200>]}}}
        Scale Curve=&Multiply_Float(A=0.2, B=&Sine(Normalized Angle=$Emitter.Age, Period=1.2, Scale=0.1, Bias=0.1))
       )
       Passthrough Float=1000
       Spawn Group Mask=$Emitter.SpawnGroup
      ))
    ))
  ]
  rendererClasses [&NiagaraLightRendererProperties]
  renderers [{rendererClass=&NiagaraLightRendererProperties, rendererIndex=0}]
 }
}
summary {
 systemName=NS_Proto_Statue_Flare
 emitters [
  {emitterName=SingleLoopingParticle, rendererClasses [&NiagaraSpriteRendererProperties]}
  {emitterName=light, rendererClasses [&NiagaraLightRendererProperties]}
  {emitterName=Glow, rendererClasses [&NiagaraSpriteRendererProperties]}
 ]
 userVariables [
  User.CubeninjaFX_Alpha {type=NiagaraFloat, defaultValue=0, description=""}
  User.IsAlt {type=NiagaraFloat, defaultValue=0, description=""}
  User.OverallSize {type=NiagaraFloat, defaultValue=0, description=""}
 ]
}
